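# Sound cue for the DangerRoom S36 door-open sound, classed under Fort_Doors.
# A single wave player feeds a Modulator that adds very slight pitch variation, with its min and max values inverted.
# All four branches of the root quality-level node share that same Modulator, so every tier sounds identical.
# It uses the shared door attenuation asset and overrides concurrency to prevent new instances once the limit is reached.
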SoundCue /Game/Sounds/Fort_Doors/World/DangerRoom_S36_DoorOpen_Cue
root=QualityLevel_0
nodes [
 {id=Modulator_0, children [WavePlayer_0], properties {pitchMax=0.99, pitchMin=1.01}}
 {id=QualityLevel_0, children [Modulator_0,Modulator_0,Modulator_0,Modulator_0]}
 {id=WavePlayer_0, properties {soundWaveAssetPtr=&/DangerRoom_S36_DoorOpen}}
]
properties {
 attenuationSettings=/Game/Sounds/Attenuation/Fort_Doors_Att
 bOverrideConcurrency=true
 maxDistance=2400
 soundClassObject=&Fort_Doors
 volumeMultiplier=1
 attenuationOverrides {falloffDistance=1600}
 concurrencyOverrides {resolutionRule=PreventNew}
}
