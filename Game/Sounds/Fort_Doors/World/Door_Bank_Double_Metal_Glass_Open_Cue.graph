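# Sound cue for the opening of a double metal-and-glass bank door in the world.
# A single wave player feeds a Modulator left at its default settings.
# The cue routes to the Fort_Doors sound class and uses a shared double-door attenuation asset with less reverb.
# A concurrency override allows only a couple of simultaneous plays and blocks new ones beyond that limit.
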SoundCue /Game/Sounds/Fort_Doors/World/Door_Bank_Double_Metal_Glass_Open_Cue
root=Modulator_0
nodes [
 {id=Modulator_0, children [WavePlayer_1]}
 {id=WavePlayer_1, properties {soundWaveAssetPtr=&/Door_Bank_Double_Metal_Glass_Open_01}}
]
properties {
 attenuationSettings=/Game/Sounds/Attenuation/Fort_Doors_Double_Less_Reverb_Att
 bOverrideConcurrency=true
 maxDistance=2500
 soundClassObject=&Fort_Doors
 volumeMultiplier=1
 concurrencyOverrides {maxCount=2, resolutionRule=PreventNew}
}
