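# Sound cue for a bunker door closing, classed under Fort_Doors.
# A single wave player feeds the door_bunker_close wave through a modulator at the root.
# It plays at reduced volume with the shared fort door attenuation asset.
# A concurrency override blocks new instances rather than cutting off playing ones when the limit is reached.
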SoundCue /Game/Sounds/Fort_Doors/World/Door_Bunker_Close_Cue
root=Modulator_1
nodes [{id=Modulator_1, children [WavePlayer_6]}, {id=WavePlayer_6, properties {soundWaveAssetPtr=&/door_bunker_close}}]
properties {
 attenuationSettings=/Game/Sounds/Attenuation/Fort_Doors_Att
 bOverrideConcurrency=true
 maxDistance=2400
 soundClassObject=&Fort_Doors
 volumeMultiplier=0.5
 attenuationOverrides {falloffDistance=1600}
 concurrencyOverrides {resolutionRule=PreventNew}
}
