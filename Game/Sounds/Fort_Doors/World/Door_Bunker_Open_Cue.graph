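# Sound cue for a bunker door opening in the world, in the Fort_Doors sound class.
# A single wave player for door_bunker_open feeds a Modulator at the root, with the cue's volume reduced.
# It uses the shared Fort_Doors attenuation asset and overrides concurrency so new instances are prevented rather than cutting off playing ones once the limit is reached.
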SoundCue /Game/Sounds/Fort_Doors/World/Door_Bunker_Open_Cue
root=Modulator_2
nodes [{id=Modulator_2, children [WavePlayer_7]}, {id=WavePlayer_7, properties {soundWaveAssetPtr=&/door_bunker_open}}]
properties {
 attenuationSettings=/Game/Sounds/Attenuation/Fort_Doors_Att
 bOverrideConcurrency=true
 maxDistance=2400
 soundClassObject=&Fort_Doors
 volumeMultiplier=0.5
 attenuationOverrides {falloffDistance=1600}
 concurrencyOverrides {resolutionRule=PreventNew}
}
