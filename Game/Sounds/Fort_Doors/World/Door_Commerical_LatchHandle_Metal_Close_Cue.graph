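# Sound cue for a metal latch-handle commercial door closing, in the Fort_Doors sound class.
# It picks one of two close recordings with equal weight, then passes it through a modulator for variation.
# A reduced-reverb, smaller-radius door attenuation spatializes it, and volume is slightly lowered.
# A concurrency override prevents new instances from starting once the limit is reached.
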SoundCue /Game/Sounds/Fort_Doors/World/Door_Commerical_LatchHandle_Metal_Close_Cue
root=Attenuation_0
nodes [
 {id=Attenuation_0, children [Modulator_0], properties {attenuationSettings=/Game/Sounds/Attenuation/Fort_Doors_Less_Reverb_Att_Smaller_Rad}}
 {id=Modulator_0, children [Random_0]}
 {id=Random_0, children [WavePlayer_1,WavePlayer_2], properties {weights [1,1]}}
 {id=WavePlayer_1, properties {soundWaveAssetPtr=&/Door_Commerical_LatchHandle_Metal_Close_01}}
 {id=WavePlayer_2, properties {soundWaveAssetPtr=&/Door_Commerical_LatchHandle_Metal_Close_02}}
]
properties {
 attenuationSettings=&Fort_Doors_Less_Reverb_Att_Smaller_Rad
 bOverrideConcurrency=true
 maxDistance=1400
 soundClassObject=&Fort_Doors
 volumeMultiplier=0.8
 concurrencyOverrides {resolutionRule=PreventNew}
}
