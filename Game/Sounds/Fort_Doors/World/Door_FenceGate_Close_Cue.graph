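# Sound cue for a fence gate closing in the world, using the door attenuation and sound class.
# A quality level node selects one of four branches, two layered and two playing a single pre-optimized close wave through a modulator.
# The layered branches mix a boosted, modulated random gate-open sound with a quieter, delayed metal sign and chainlink impact for the latch hit, one with fewer variations.
# Concurrency is overridden to prevent new instances at the limit.
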SoundCue /Game/Sounds/Fort_Doors/World/Door_FenceGate_Close_Cue
root=QualityLevel_0
nodes [
 {id=Delay_0, children [Mixer_1], properties {delayMax=0.5, delayMin=0.5}}
 {id=Delay_1, children [Mixer_3], properties {delayMax=0.5, delayMin=0.5}}
 {id=Mixer_0, children [Modulator_0,Delay_0], properties {inputVolume [2,0.5]}}
 {id=Mixer_1, children [WavePlayer_5,Random_1], properties {inputVolume [1,1]}}
 {id=Mixer_2, children [Modulator_3,Delay_1], properties {inputVolume [2,0.5]}}
 {id=Mixer_3, children [WavePlayer_13, WavePlayer_8], properties {inputVolume [1,1]}}
 {id=Modulator_0, children [Random_0]}
 {id=Modulator_1, children [WavePlayer_6]}
 {id=Modulator_2, children [WavePlayer_7]}
 {id=Modulator_3, children [Random_2]}
 {id=QualityLevel_0, children [Mixer_0,Mixer_2,Modulator_1,Modulator_2]}
 {id=Random_0, children [WavePlayer_1,WavePlayer_2,WavePlayer_3], properties {weights [1,1,1]}}
 {id=Random_1, children [WavePlayer_0,WavePlayer_4], properties {weights [1,1]}}
 {id=Random_2, children [WavePlayer_9, WavePlayer_10], properties {weights [1,1]}}
 {id=WavePlayer_0, properties {soundWaveAssetPtr=/Game/Sounds/Fort_Impact_Sounds/Demolition/Fence/fence_chainlink_impact_ST_01}}
 {id=WavePlayer_1, properties {soundWaveAssetPtr=&/Door_Open_FenceGate01}}
 {id=WavePlayer_10, properties {soundWaveAssetPtr=&/Door_Open_FenceGate02}}
 {id=WavePlayer_13, properties {soundWaveAssetPtr=/Game/Sounds/Fort_Impact_Sounds/Demolition/Metal/Metal_Sign_Impact02}}
 {id=WavePlayer_2, properties {soundWaveAssetPtr=&Door_Open_FenceGate02}}
 {id=WavePlayer_3, properties {soundWaveAssetPtr=&/Door_Open_FenceGate03}}
 {id=WavePlayer_4, properties {soundWaveAssetPtr=/Game/Sounds/Fort_Impact_Sounds/Demolition/Fence/fence_chainlink_impact_ST_02}}
 {id=WavePlayer_5, properties {soundWaveAssetPtr=&Metal_Sign_Impact02}}
 {id=WavePlayer_6, properties {soundWaveAssetPtr=&/Door_FenceGate_Close_opt}}
 {id=WavePlayer_7, properties {soundWaveAssetPtr=&Door_FenceGate_Close_opt}}
 {id=WavePlayer_8, properties {soundWaveAssetPtr=&fence_chainlink_impact_ST_01}}
 {id=WavePlayer_9, properties {soundWaveAssetPtr=&Door_Open_FenceGate01}}
]
properties {
 attenuationSettings=/Game/Sounds/Attenuation/Fort_Doors_Att
 bOverrideConcurrency=true
 maxDistance=2400
 soundClassObject=&Fort_Doors
 volumeMultiplier=0.5
 attenuationOverrides {falloffDistance=1600}
 concurrencyOverrides {resolutionRule=PreventNew}
}
