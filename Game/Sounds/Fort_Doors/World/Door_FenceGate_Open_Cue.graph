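# Sound cue for a fence gate door opening in the world, with door attenuation and a concurrency override that prevents new sounds at the limit.
# A quality node switches between four branches: the first two mix a modulated random gate-open wave with a chainlink impact that an enveloper fades in.
# The second branch uses fewer variations, and the last two play only a modulated optimized wave.
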SoundCue /Game/Sounds/Fort_Doors/World/Door_FenceGate_Open_Cue
root=QualityLevel_0
nodes [
 {
  id=Enveloper_0
  children [Random_1]
  properties {pitchCurve {editorCurveData {keys [{time=0}]}}, volumeCurve {editorCurveData {keys [<at 0.0022 0>,<at 1 1.002>]}}}
 }
 {
  id=Enveloper_1
  children [WavePlayer_5]
  properties {pitchCurve {editorCurveData {keys [{time=0}]}}, volumeCurve {editorCurveData {keys [<at 0.0022 0>,<at 1 1.002>]}}}
 }
 {id=Mixer_0, children [Modulator_0,Enveloper_0], properties {inputVolume [1.5,1]}}
 {id=Mixer_1, children [Modulator_1,Enveloper_1], properties {inputVolume [1.5,1]}}
 {id=Modulator_0, children [Random_0]}
 {id=Modulator_1, children [Random_2]}
 {id=Modulator_2, children [WavePlayer_10]}
 {id=Modulator_3, children [WavePlayer_11]}
 {id=QualityLevel_0, children [Mixer_0,Mixer_1,Modulator_2,Modulator_3]}
 {id=Random_0, children [WavePlayer_1,WavePlayer_2,WavePlayer_3], properties {weights [1,1,1]}}
 {id=Random_1, children [WavePlayer_0,WavePlayer_4], properties {weights [1,1]}}
 {id=Random_2, children [WavePlayer_6,WavePlayer_7], properties {weights [1,1]}}
 {id=WavePlayer_0, properties {soundWaveAssetPtr=/Game/Sounds/Fort_Impact_Sounds/Demolition/Fence/fence_chainlink_impact_ST_01}}
 {id=WavePlayer_1, properties {soundWaveAssetPtr=&/Door_Open_FenceGate01}}
 {id=WavePlayer_10, properties {soundWaveAssetPtr=&/Door_Open_FenceGate_opt}}
 {id=WavePlayer_11, properties {soundWaveAssetPtr=&Door_Open_FenceGate_opt}}
 {id=WavePlayer_2, properties {soundWaveAssetPtr=&/Door_Open_FenceGate02}}
 {id=WavePlayer_3, properties {soundWaveAssetPtr=&/Door_Open_FenceGate03}}
 {id=WavePlayer_4, properties {soundWaveAssetPtr=/Game/Sounds/Fort_Impact_Sounds/Demolition/Fence/fence_chainlink_impact_ST_02}}
 {id=WavePlayer_5, properties {soundWaveAssetPtr=&fence_chainlink_impact_ST_01}}
 {id=WavePlayer_6, properties {soundWaveAssetPtr=&Door_Open_FenceGate01}}
 {id=WavePlayer_7, properties {soundWaveAssetPtr=&Door_Open_FenceGate02}}
]
properties {
 attenuationSettings=/Game/Sounds/Attenuation/Fort_Doors_Att
 bOverrideConcurrency=true
 maxDistance=2400
 soundClassObject=&Fort_Doors
 volumeMultiplier=0.5
 attenuationOverrides {falloffDistance=1600}
 concurrencyOverrides {resolutionRule=PreventNew}
}
