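# Sound cue for a fast-opening garage door in the world door set.
# A random node picks between two equally weighted wave variants, then passes the result through a modulator at default settings.
# The cue plays at boosted volume in the doors sound class with a reduced-reverb door attenuation.
# It primes on load and overrides concurrency to block new instances once the limit is reached.
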SoundCue /Game/Sounds/Fort_Doors/World/Door_Garage01_Open_Fast_Cue
root=Modulator_0
nodes [
 {id=Modulator_0, children [Random_0]}
 {id=Random_0, children [WavePlayer_1,WavePlayer_2], properties {weights [1,1]}}
 {id=WavePlayer_1, properties {soundWaveAssetPtr=&/Door_Garage_Open_01a_Fast}}
 {id=WavePlayer_2, properties {soundWaveAssetPtr=&/Door_Garage_Open_01b_Fast}}
]
properties {
 attenuationSettings=/Game/Sounds/Attenuation/Fort_Doors_Less_Reverb_Att
 bOverrideConcurrency=true
 bPrimeOnLoad=true
 maxDistance=2400
 soundClassObject=&Fort_Doors
 volumeMultiplier=1.8
 concurrencyOverrides {resolutionRule=PreventNew}
}
