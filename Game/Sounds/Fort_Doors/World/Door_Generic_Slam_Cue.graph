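# Sound cue for a generic world door slam, routed through the Fort_WeaponMelee_PSM sound class.
# A Random node picks one of three equally weighted slam waves and is excluded from branch culling, then feeds a Modulator with volume variation pinned flat.
# Playback uses the shared Fort_Doors slam-open attenuation asset and overrides concurrency to cap simultaneous instances.
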SoundCue /Game/Sounds/Fort_Doors/World/Door_Generic_Slam_Cue
root=Modulator_3
nodes [
 {id=Modulator_3, children [Random_1], properties {volumeMax=1, volumeMin=1}}
 {id=Random_1, children [WavePlayer_7,WavePlayer_8,WavePlayer_9], properties {bShouldExcludeFromBranchCulling=true, weights [1,1,1]}}
 {id=WavePlayer_7, properties {soundWaveAssetPtr=&/Door_Generic_Slam_01}}
 {id=WavePlayer_8, properties {soundWaveAssetPtr=&/Door_Generic_Slam_02}}
 {id=WavePlayer_9, properties {soundWaveAssetPtr=&/Door_Generic_Slam_03}}
]
properties {
 attenuationSettings=/Game/Sounds/Attenuation/Fort_Doors_Slam_Open_Att
 bOverrideConcurrency=true
 maxDistance=2500
 soundClassObject=&Fort_WeaponMelee_PSM
 volumeMultiplier=1.25
 attenuationOverrides {falloffDistance=1600}
 concurrencyOverrides {maxCount=6}
}
