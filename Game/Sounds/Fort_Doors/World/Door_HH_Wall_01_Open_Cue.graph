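# Sound cue for the opening of a Hunters Haven wall door in the world.
# A single wave player feeds the output directly with one door-open wave, with no randomization or modulation.
# It routes to the Fort_Doors sound class, uses the shared doors attenuation at reduced volume, and overrides concurrency to prevent new instances once the limit is reached.
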SoundCue /Game/Sounds/Fort_Doors/World/Door_HH_Wall_01_Open_Cue
root=WavePlayer_0
nodes [{id=WavePlayer_0, properties {soundWaveAssetPtr=&/Door_S15_HuntersHaven_01_Open_01}}]
properties {
 attenuationSettings=/Game/Sounds/Attenuation/Fort_Doors_Att
 bOverrideConcurrency=true
 maxDistance=2400
 soundClassObject=&Fort_Doors
 volumeMultiplier=0.7
 concurrencyOverrides {resolutionRule=PreventNew}
}
