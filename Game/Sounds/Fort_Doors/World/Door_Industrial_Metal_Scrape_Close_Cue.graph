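# Sound cue for the closing scrape of an industrial metal world door, in the Fort_Doors sound class.
# A Random node picks between two scrape-close wave variants at equal weight, then a Modulator applies variation before output.
# It uses the reduced-reverb door attenuation and overrides concurrency to prevent new instances rather than cut off playing ones once the limit is reached.
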SoundCue /Game/Sounds/Fort_Doors/World/Door_Industrial_Metal_Scrape_Close_Cue
root=Modulator_0
nodes [
 {id=Modulator_0, children [Random_0]}
 {id=Random_0, children [WavePlayer_0,WavePlayer_3], properties {weights [1,1]}}
 {id=WavePlayer_0, properties {soundWaveAssetPtr=&/Door_Industrial_Metal_Scrape_Close_01}}
 {id=WavePlayer_3, properties {soundWaveAssetPtr=&/Door_Industrial_Metal_Scrape_Close_02}}
]
properties {
 attenuationSettings=/Game/Sounds/Attenuation/Fort_Doors_Less_Reverb_Att
 bOverrideConcurrency=true
 maxDistance=2400
 soundClassObject=&Fort_Doors
 volumeMultiplier=1
 concurrencyOverrides {resolutionRule=PreventNew}
}
