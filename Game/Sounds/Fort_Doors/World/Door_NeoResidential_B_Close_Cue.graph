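# Sound cue for the closing sound of the NeoResidential B world door.
# A single wave player feeds the Door_NeoResidential_B_Close sound wave straight to the output, with no randomization or modulation.
# It uses the shared Fort_Doors attenuation asset and sound class, with a max distance of 2400.
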SoundCue /Game/Sounds/Fort_Doors/World/Door_NeoResidential_B_Close_Cue
root=WavePlayer_0
nodes [{id=WavePlayer_0, properties {soundWaveAssetPtr=&/Door_NeoResidential_B_Close}}]
properties {
 attenuationSettings=/Game/Sounds/Attenuation/Fort_Doors_Att
 maxDistance=2400
 soundClassObject=&Fort_Doors
}
