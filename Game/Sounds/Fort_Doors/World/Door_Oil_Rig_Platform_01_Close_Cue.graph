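# Sound cue for closing the oil rig platform door in the Fort_Doors world sound set.
# A single wave player feeds the output directly, reusing the metal interior office door close sample.
# It plays at reduced volume through the shared Fort_Doors attenuation and sound class, with no randomization or modulation.
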SoundCue /Game/Sounds/Fort_Doors/World/Door_Oil_Rig_Platform_01_Close_Cue
root=WavePlayer_5
nodes [{id=WavePlayer_5, properties {soundWaveAssetPtr=&/Door_Metal_Interior_Office_Close_01}}]
properties {
 attenuationSettings=/Game/Sounds/Attenuation/Fort_Doors_Att
 maxDistance=2400
 soundClassObject=&Fort_Doors
 volumeMultiplier=0.5
}
