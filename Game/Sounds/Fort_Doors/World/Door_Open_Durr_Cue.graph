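# Sound cue for a world door-opening sound in the Fort_Doors sound class.
# A Mixer root layers two waves at once: Door_Open_Durr at full volume and metal_door_open_D at half volume.
# Spatial falloff comes from the shared Fort_Doors_Att attenuation asset, and a concurrency override blocks new instances once the limit is reached.
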SoundCue /Game/Sounds/Fort_Doors/World/Door_Open_Durr_Cue
root=Mixer_3
nodes [
 {id=Mixer_3, children [WavePlayer_26, WavePlayer_10], properties {inputVolume [1,0.5]}}
 {id=WavePlayer_10, properties {soundWaveAssetPtr=&/metal_door_open_D}}
 {id=WavePlayer_26, properties {soundWaveAssetPtr=&/Door_Open_Durr}}
]
properties {
 attenuationSettings=/Game/Sounds/Attenuation/Fort_Doors_Att
 bOverrideConcurrency=true
 maxDistance=2400
 soundClassObject=&Fort_Doors
 attenuationOverrides {falloffDistance=1600}
 concurrencyOverrides {resolutionRule=PreventNew}
}
