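# Sound cue for a store door opening, played at reduced volume through the Fort_Doors sound class and door attenuation settings.
# A Random node picks between two equally weighted mixes, each layering Door_Open_Store over a half-volume metal_door_open_D.
# One mix adds an Enveloper that fades the store layer out early, giving a shorter variant.
# Overridden concurrency uses PreventNew, so new plays are refused once the limit is reached.
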SoundCue /Game/Sounds/Fort_Doors/World/Door_Open_Store_Cue
root=Random_0
nodes [
 {
  id=Enveloper_0
  children [WavePlayer_1]
  properties {pitchCurve {editorCurveData {keys [{time=0}]}}, volumeCurve {editorCurveData {keys [<at 0 1>,<at 0.5 1>,<at 0.7 0>]}}}
 }
 {id=Mixer_0, children [Enveloper_0,WavePlayer_2], properties {inputVolume [1,0.5]}}
 {id=Mixer_2, children [WavePlayer_22, WavePlayer_9], properties {inputVolume [1,0.5]}}
 {id=Random_0, children [Mixer_2,Mixer_0], properties {weights [1,1]}}
 {id=WavePlayer_1, properties {soundWaveAssetPtr=&/Door_Open_Store}}
 {id=WavePlayer_2, properties {soundWaveAssetPtr=&/metal_door_open_D}}
 {id=WavePlayer_22, properties {soundWaveAssetPtr=&Door_Open_Store}}
 {id=WavePlayer_9, properties {soundWaveAssetPtr=&metal_door_open_D}}
]
properties {
 attenuationSettings=/Game/Sounds/Attenuation/Fort_Doors_Att
 bOverrideConcurrency=true
 maxDistance=2400
 soundClassObject=&Fort_Doors
 volumeMultiplier=0.5
 attenuationOverrides {falloffDistance=1600}
 concurrencyOverrides {resolutionRule=PreventNew}
}
