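# Sound cue for a picket fence door opening, routed to the Fort_Doors sound class with shared door attenuation.
# A quality level root selects among four branches that use fewer variations in each successive slot.
# These are a modulated random pick of five waves, a modulated pick of two, one wave with widened pitch variation, and one unmodulated wave.
# Overridden concurrency prevents new instances once the limit is reached.
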SoundCue /Game/Sounds/Fort_Doors/World/Door_Picket_Open_Cue
root=QualityLevel_0
nodes [
 {id=Modulator_0, children [Random_0]}
 {id=Modulator_1, children [Random_2]}
 {id=Modulator_2, children [WavePlayer_10], properties {pitchMax=1.2, pitchMin=0.75}}
 {id=QualityLevel_0, children [Modulator_0, Modulator_1, Modulator_2, WavePlayer_11]}
 {id=Random_0, children [WavePlayer_0,WavePlayer_1,WavePlayer_2,WavePlayer_3,WavePlayer_4], properties {weights [1,1,1,1,1]}}
 {id=Random_2, children [WavePlayer_5,WavePlayer_6], properties {weights [1,1]}}
 {id=WavePlayer_0, properties {soundWaveAssetPtr=&/Door_PicketFence_Open_01}}
 {id=WavePlayer_1, properties {soundWaveAssetPtr=&/Door_PicketFence_Open_03}}
 {id=WavePlayer_10, properties {soundWaveAssetPtr=&Door_PicketFence_Open_01}}
 {id=WavePlayer_11, properties {soundWaveAssetPtr=&Door_PicketFence_Open_01}}
 {id=WavePlayer_2, properties {soundWaveAssetPtr=&/Door_PicketFence_Open_05}}
 {id=WavePlayer_3, properties {soundWaveAssetPtr=&/Door_PicketFence_Open_02}}
 {id=WavePlayer_4, properties {soundWaveAssetPtr=&/Door_PicketFence_Open_04}}
 {id=WavePlayer_5, properties {soundWaveAssetPtr=&Door_PicketFence_Open_01}}
 {id=WavePlayer_6, properties {soundWaveAssetPtr=&Door_PicketFence_Open_03}}
]
properties {
 attenuationSettings=/Game/Sounds/Attenuation/Fort_Doors_Att
 bOverrideConcurrency=true
 maxDistance=2400
 soundClassObject=&Fort_Doors
 volumeMultiplier=0.7
 attenuationOverrides {falloffDistance=1600}
 concurrencyOverrides {resolutionRule=PreventNew}
}
