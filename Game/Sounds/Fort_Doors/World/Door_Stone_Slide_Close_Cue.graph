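# Sound cue for a stone door sliding closed, assigned to the Fort_Doors sound class.
# A single wave player feeds a Modulator at default settings, so the one recording gets slight pitch and volume variation per play.
# The cue references the shared Fort_Doors attenuation asset, carries a slight volume boost, and overrides concurrency to prevent new instances once the limit is reached.
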SoundCue /Game/Sounds/Fort_Doors/World/Door_Stone_Slide_Close_Cue
root=Modulator_1
nodes [
 {id=Modulator_1, children [WavePlayer_2]}
 {id=WavePlayer_2, properties {soundWaveAssetPtr=&/Stone_Door_Slide_Close_01}}
]
properties {
 attenuationSettings=/Game/Sounds/Attenuation/Fort_Doors_Att
 bOverrideConcurrency=true
 maxDistance=2400
 soundClassObject=&Fort_Doors
 volumeMultiplier=1.2
 attenuationOverrides {falloffDistance=1600}
 concurrencyOverrides {resolutionRule=PreventNew}
}
