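# Sound cue for a store door closing, part of the Fort_Doors world sounds.
# A single wave player at the root plays the Door_Store_Close_mix wave at reduced volume, with no randomization or layering.
# It references the shared Fort_Doors attenuation asset and sound class, and overrides concurrency so new instances are blocked rather than cutting off playing ones.
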SoundCue /Game/Sounds/Fort_Doors/World/Door_Store_Close_Cue
root=WavePlayer_0
nodes [{id=WavePlayer_0, properties {soundWaveAssetPtr=&/Door_Store_Close_mix}}]
properties {
 attenuationSettings=/Game/Sounds/Attenuation/Fort_Doors_Att
 bOverrideConcurrency=true
 maxDistance=2400
 soundClassObject=&Fort_Doors
 volumeMultiplier=0.5
 attenuationOverrides {falloffDistance=1600}
 concurrencyOverrides {resolutionRule=PreventNew}
}
