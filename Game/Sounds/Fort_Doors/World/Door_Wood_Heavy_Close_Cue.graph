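# Sound cue for a heavy wooden door closing, assigned to the Fort_Doors sound class.
# A Random node picks one of two close recordings with equal weight and feeds it through a Modulator at default settings, with the cue's volume reduced.
# It uses the shared Fort_Doors attenuation asset for spatial falloff.
# Its own concurrency override blocks new instances once the limit is reached instead of cutting off playing ones.
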SoundCue /Game/Sounds/Fort_Doors/World/Door_Wood_Heavy_Close_Cue
root=Modulator_2
nodes [
 {id=Modulator_2, children [Random_0]}
 {id=Random_0, children [WavePlayer_0,WavePlayer_1], properties {weights [1,1]}}
 {id=WavePlayer_0, properties {soundWaveAssetPtr=&/Door_Wood_Heavy_Close_01}}
 {id=WavePlayer_1, properties {soundWaveAssetPtr=&/Door_Wood_Heavy_Close_02}}
]
properties {
 attenuationSettings=/Game/Sounds/Attenuation/Fort_Doors_Att
 bOverrideConcurrency=true
 maxDistance=2400
 soundClassObject=&Fort_Doors
 volumeMultiplier=0.5
 attenuationOverrides {falloffDistance=1600}
 concurrencyOverrides {resolutionRule=PreventNew}
}
