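# Sound cue for the Yacht "Cat" door opening sound in the world.
# A single wave player feeds the output directly with a metal swinging door open sample, with no randomization or modulation nodes.
# It uses the shared Fort_Doors attenuation asset and the Fort_Doors sound class.
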SoundCue /Game/Sounds/Fort_Doors/World/Door_Yacht_Cat_Open_Cue
root=WavePlayer_1
nodes [{id=WavePlayer_1, properties {soundWaveAssetPtr=&/Door_Metal_Swinging_Cat_Open_01}}]
properties {
 attenuationSettings=/Game/Sounds/Attenuation/Fort_Doors_Att
 maxDistance=2400
 soundClassObject=&Fort_Doors
 volumeMultiplier=1
}
