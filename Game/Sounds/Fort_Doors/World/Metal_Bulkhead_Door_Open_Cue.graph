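# Sound cue for a metal bulkhead door opening in the world, assigned to the Fort_Doors sound class.
# A quality-level node sends its first two levels to an equal-weight random pick between two recordings, and its last two to the first recording alone.
# It uses door attenuation without occlusion, and its concurrency override prevents new instances once the limit is reached.
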SoundCue /Game/Sounds/Fort_Doors/World/Metal_Bulkhead_Door_Open_Cue
root=QualityLevel_0
nodes [
 {id=QualityLevel_0, children [Random_1,Random_1,WavePlayer_2,WavePlayer_2]}
 {id=Random_1, children [WavePlayer_0,WavePlayer_1], properties {weights [1,1]}}
 {id=WavePlayer_0, properties {soundWaveAssetPtr=&/Metal_Bulkhead_Door_Open_01}}
 {id=WavePlayer_1, properties {soundWaveAssetPtr=&/Metal_Bulkhead_Door_Open_02}}
 {id=WavePlayer_2, properties {soundWaveAssetPtr=&Metal_Bulkhead_Door_Open_01}}
]
properties {
 attenuationSettings=/Game/Sounds/Attenuation/Fort_Doors_NoOcclusion_Att
 bOverrideConcurrency=true
 maxDistance=2400
 soundClassObject=&Fort_Doors
 volumeMultiplier=0.7
 concurrencyOverrides {resolutionRule=PreventNew}
}
