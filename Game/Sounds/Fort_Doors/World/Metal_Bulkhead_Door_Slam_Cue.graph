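# Sound cue for a metal bulkhead door slam in the world, using the Fort_Doors sound class and door attenuation.
# A quality-level node sends its first two levels through a modulator to an equally weighted random pick of four slam waves.
# The last two levels use a second modulator playing only slam variation 03.
# A concurrency override prevents new instances once the limit is reached.
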SoundCue /Game/Sounds/Fort_Doors/World/Metal_Bulkhead_Door_Slam_Cue
root=QualityLevel_0
nodes [
 {id=Modulator_0, children [Random_1]}
 {id=Modulator_1, children [WavePlayer_8]}
 {id=QualityLevel_0, children [Modulator_0,Modulator_0,Modulator_1,Modulator_1]}
 {id=Random_1, children [WavePlayer_2,WavePlayer_5,WavePlayer_6,WavePlayer_7], properties {weights [1,1,1,1]}}
 {id=WavePlayer_2, properties {soundWaveAssetPtr=&/Metal_Bulkhead_Door_Slam_01}}
 {id=WavePlayer_5, properties {soundWaveAssetPtr=&/Metal_Bulkhead_Door_Slam_02}}
 {id=WavePlayer_6, properties {soundWaveAssetPtr=&/Metal_Bulkhead_Door_Slam_03}}
 {id=WavePlayer_7, properties {soundWaveAssetPtr=&/Metal_Bulkhead_Door_Slam_04}}
 {id=WavePlayer_8, properties {soundWaveAssetPtr=&Metal_Bulkhead_Door_Slam_03}}
]
properties {
 attenuationSettings=/Game/Sounds/Attenuation/Fort_Doors_Att
 bOverrideConcurrency=true
 maxDistance=2400
 soundClassObject=&Fort_Doors
 volumeMultiplier=1.2
 concurrencyOverrides {resolutionRule=PreventNew}
}
