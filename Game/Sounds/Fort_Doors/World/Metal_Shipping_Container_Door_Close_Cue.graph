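# Sound cue for a metal shipping container door closing in the world, built on BurnTyres double-door recordings.
# A quality-level node feeds the first two slots from an equally weighted random pick of two close variations, and the other two from a single fixed variation.
# It uses the shared Fort_Doors sound class and attenuation, with a concurrency override that prevents new instances once the limit is reached.
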SoundCue /Game/Sounds/Fort_Doors/World/Metal_Shipping_Container_Door_Close_Cue
root=QualityLevel_0
nodes [
 {id=QualityLevel_0, children [Random_0,Random_0,WavePlayer_6,WavePlayer_6]}
 {id=Random_0, children [WavePlayer_3,WavePlayer_4], properties {weights [1,1]}}
 {id=WavePlayer_3, properties {soundWaveAssetPtr=&/BurnTyres_DoubleDoor_Close_01}}
 {id=WavePlayer_4, properties {soundWaveAssetPtr=&/BurnTyres_DoubleDoor_Close_02}}
 {id=WavePlayer_6, properties {soundWaveAssetPtr=&BurnTyres_DoubleDoor_Close_01}}
]
properties {
 attenuationSettings=/Game/Sounds/Attenuation/Fort_Doors_Att
 bOverrideConcurrency=true
 maxDistance=2400
 soundClassObject=&Fort_Doors
 volumeMultiplier=0.8
 concurrencyOverrides {resolutionRule=PreventNew}
}
