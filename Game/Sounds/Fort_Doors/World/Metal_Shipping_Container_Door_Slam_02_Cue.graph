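# Sound cue for a metal shipping container door slam, in the Fort_Doors sound class.
# A single wave player feeds the BurnTyres_DoubleDoor_Slam_01 recording through a Modulator left at its default settings.
# It uses the shared door attenuation without occlusion and overrides concurrency so new instances are refused once the limit is reached.
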SoundCue /Game/Sounds/Fort_Doors/World/Metal_Shipping_Container_Door_Slam_02_Cue
root=Modulator_3
nodes [
 {id=Modulator_3, children [WavePlayer_0]}
 {id=WavePlayer_0, properties {soundWaveAssetPtr=&/BurnTyres_DoubleDoor_Slam_01}}
]
properties {
 attenuationSettings=/Game/Sounds/Attenuation/Fort_Doors_NoOcclusion_Att
 bOverrideConcurrency=true
 maxDistance=2400
 soundClassObject=&Fort_Doors
 volumeMultiplier=1
 attenuationOverrides {falloffDistance=1600}
 concurrencyOverrides {resolutionRule=PreventNew}
}
