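# Sound cue that plays the opening sound for a TechHQ world door, in the Fort_Doors sound class.
# A quality-level switch sends its first two levels to a modulated random pick between two equally weighted open waves.
# Its last two levels share a second modulator playing only an Open01 wave.
# It uses no-occlusion door attenuation and overrides concurrency to prevent new instances once the limit is reached.
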SoundCue /Game/Sounds/Fort_Doors/World/SC_Door_TechHQ_01_Open
root=QualityLevel_0
nodes [
 {id=Modulator_0, children [Random_0]}
 {id=Modulator_1, children [WavePlayer_1]}
 {id=QualityLevel_0, children [Modulator_0,Modulator_0,Modulator_1,Modulator_1]}
 {id=Random_0, children [WavePlayer_0,WavePlayer_2], properties {weights [1,1]}}
 {id=WavePlayer_0, properties {soundWaveAssetPtr=&/Door_TechHQ_Open01}}
 {id=WavePlayer_1, properties {soundWaveAssetPtr=&Door_TechHQ_Open01}}
 {id=WavePlayer_2, properties {soundWaveAssetPtr=&/Door_TechHQ_Open02}}
]
properties {
 attenuationSettings=/Game/Sounds/Attenuation/Fort_Doors_NoOcclusion_Att
 bOverrideConcurrency=true
 maxDistance=2400
 soundClassObject=&Fort_Doors
 volumeMultiplier=1
 concurrencyOverrides {resolutionRule=PreventNew}
}
