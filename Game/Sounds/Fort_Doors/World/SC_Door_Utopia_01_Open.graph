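# Sound cue for the opening sound of the Utopia 01 world door, in the Fort_Doors sound class.
# A quality-level root sends its first two levels to one wave and its last two to another like-named wave asset, each through a default-setting Modulator.
# Playback uses the shared no-occlusion door attenuation at reduced volume, with a concurrency override that prevents new instances once the limit is reached.
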SoundCue /Game/Sounds/Fort_Doors/World/SC_Door_Utopia_01_Open
root=QualityLevel_0
nodes [
 {id=Modulator_0, children [WavePlayer_0]}
 {id=Modulator_1, children [WavePlayer_1]}
 {id=QualityLevel_0, children [Modulator_0,Modulator_0,Modulator_1,Modulator_1]}
 {id=WavePlayer_0, properties {soundWaveAssetPtr=&/Door_Utopia_01_Open}}
 {id=WavePlayer_1, properties {soundWaveAssetPtr=&Door_Utopia_01_Open}}
]
properties {
 attenuationSettings=/Game/Sounds/Attenuation/Fort_Doors_NoOcclusion_Att
 bOverrideConcurrency=true
 maxDistance=2400
 soundClassObject=&Fort_Doors
 volumeMultiplier=0.45
 concurrencyOverrides {resolutionRule=PreventNew}
}
