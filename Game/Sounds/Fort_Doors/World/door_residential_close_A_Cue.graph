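# Sound cue for a residential door closing in the world, played through the Fort_Doors sound class and attenuation.
# The two higher quality tiers mix a random E-series sound, enveloped to fade out by about 0.6 seconds, with a delayed, modulated random A-series sound that overlaps its tail.
# The two lower tiers use a cheaper branch that randomly picks one of two modulated optimized waves.
# A concurrency override prevents new instances once the limit is reached.
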SoundCue /Game/Sounds/Fort_Doors/World/door_residential_close_A_Cue
root=QualityLevel_0
nodes [
 {id=Delay_1, children [Modulator_0], properties {delayMax=0.4, delayMin=0.4}}
 {
  id=Enveloper_1
  children [Random_1]
  properties {
   pitchCurve {editorCurveData {keys [{time=0}]}}
   volumeCurve {editorCurveData {keys [
      <at 0.0029 1>
      {arriveTangent=-1.7918, interpMode=RCIM_Cubic, leaveTangent=-1.7918, time=0.4344, value=0.9538}
      <at 0.6052 -0.0792 RCIM_Cubic>
     ]}}
  }
 }
 {id=Mixer_2, children [Delay_1,Enveloper_1], properties {inputVolume [1,1]}}
 {id=Modulator_0, children [Random_0]}
 {id=Modulator_2, children [Random_4]}
 {id=QualityLevel_0, children [Mixer_2,Mixer_2,Modulator_2,Modulator_2]}
 {id=Random_0, children [WavePlayer_3,WavePlayer_4,WavePlayer_5], properties {weights [1,1,1]}}
 {id=Random_1, children [WavePlayer_7,WavePlayer_8,WavePlayer_9], properties {weights [1,1,1]}}
 {id=Random_4, children [WavePlayer_0, WavePlayer_14], properties {bShouldExcludeFromBranchCulling=true, weights [1,1]}}
 {id=WavePlayer_0, properties {soundWaveAssetPtr=&/door_residential_close_A_opt}}
 {id=WavePlayer_14, properties {soundWaveAssetPtr=&/door_residential_close_A2_opt}}
 {id=WavePlayer_3, properties {soundWaveAssetPtr=&/door_residential_close_A1}}
 {id=WavePlayer_4, properties {soundWaveAssetPtr=&/door_residential_close_A2}}
 {id=WavePlayer_5, properties {soundWaveAssetPtr=&/door_residential_close_A3}}
 {id=WavePlayer_7, properties {soundWaveAssetPtr=&/door_residential_close_E1}}
 {id=WavePlayer_8, properties {soundWaveAssetPtr=&/door_residential_close_E2}}
 {id=WavePlayer_9, properties {soundWaveAssetPtr=&/door_residential_close_E3}}
]
properties {
 attenuationSettings=/Game/Sounds/Attenuation/Fort_Doors_Att
 bOverrideConcurrency=true
 maxDistance=2400
 soundClassObject=&Fort_Doors
 volumeMultiplier=0.6
 concurrencyOverrides {resolutionRule=PreventNew}
}
