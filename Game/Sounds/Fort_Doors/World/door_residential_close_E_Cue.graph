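# Sound cue for the closing sound of residential door variant E, played in the Fort_Doors sound class with shared door attenuation.
# A quality-level node sends the first two levels to a three-wave random pool and the last two to a two-wave pool excluded from branch culling.
# Each pool passes through a Modulator for pitch variation.
# The cue overrides concurrency to block new instances at the limit.
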SoundCue /Game/Sounds/Fort_Doors/World/door_residential_close_E_Cue
root=QualityLevel_0
nodes [
 {id=Modulator_0, children [Random_0], properties {pitchMax=1.1}}
 {id=Modulator_1, children [Random_1], properties {pitchMax=1.1}}
 {id=QualityLevel_0, children [Modulator_0,Modulator_0,Modulator_1,Modulator_1]}
 {id=Random_0, children [WavePlayer_3,WavePlayer_4,WavePlayer_5], properties {weights [1,1,1]}}
 {id=Random_1, children [WavePlayer_0,WavePlayer_1], properties {bShouldExcludeFromBranchCulling=true, weights [1,1]}}
 {id=WavePlayer_0, properties {soundWaveAssetPtr=&/door_residential_close_E1}}
 {id=WavePlayer_1, properties {soundWaveAssetPtr=&/door_residential_close_E2}}
 {id=WavePlayer_3, properties {soundWaveAssetPtr=&door_residential_close_E1}}
 {id=WavePlayer_4, properties {soundWaveAssetPtr=&door_residential_close_E2}}
 {id=WavePlayer_5, properties {soundWaveAssetPtr=&/door_residential_close_E3}}
]
properties {
 attenuationSettings=/Game/Sounds/Attenuation/Fort_Doors_Att
 bOverrideConcurrency=true
 maxDistance=2400
 soundClassObject=&Fort_Doors
 volumeMultiplier=0.6
 concurrencyOverrides {resolutionRule=PreventNew}
}
