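# Residential door close sound cue (G variant) in the Fort_Doors class, with door attenuation and concurrency that prevents new instances.
# The first two quality levels layer a random, pitched-down E close, faded out near half a second, with a random, modulated G close delayed half a second.
# The last two quality levels instead play one of two modulated "opt" waves, skipping the layering.
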SoundCue /Game/Sounds/Fort_Doors/World/door_residential_close_G_Cue
root=QualityLevel_0
nodes [
 {id=Delay_6, children [Modulator_0], properties {delayMax=0.5, delayMin=0.5}}
 {
  id=Enveloper_10
  children [Random_2]
  properties {pitchCurve {editorCurveData {keys [<at 0 0.8577>]}}, volumeCurve {editorCurveData {keys [<at 0 1>,<at 0.4545 0.9909>,<at 0.5803 -0.0097>]}}}
 }
 {id=Mixer_4, children [Delay_6,Enveloper_10], properties {inputVolume [1,1.2]}}
 {id=Modulator_0, children [Random_1]}
 {id=Modulator_2, children [Random_5]}
 {id=QualityLevel_0, children [Mixer_4,Mixer_4,Modulator_2,Modulator_2]}
 {id=Random_1, children [WavePlayer_0,WavePlayer_1,WavePlayer_2], properties {weights [1,1,1]}}
 {id=Random_2, children [WavePlayer_3,WavePlayer_4,WavePlayer_5], properties {weights [1,1,1]}}
 {id=Random_5, children [WavePlayer_12, WavePlayer_14], properties {bShouldExcludeFromBranchCulling=true, weights [1,1]}}
 {id=WavePlayer_0, properties {soundWaveAssetPtr=&/door_residential_close_G1}}
 {id=WavePlayer_1, properties {soundWaveAssetPtr=&/door_residential_close_G3}}
 {id=WavePlayer_12, properties {soundWaveAssetPtr=&/door_residential_close_G_opt}}
 {id=WavePlayer_14, properties {soundWaveAssetPtr=&/door_residential_close_G2_opt}}
 {id=WavePlayer_2, properties {soundWaveAssetPtr=&/door_residential_close_G2}}
 {id=WavePlayer_3, properties {soundWaveAssetPtr=&/door_residential_close_E1}}
 {id=WavePlayer_4, properties {soundWaveAssetPtr=&/door_residential_close_E2}}
 {id=WavePlayer_5, properties {soundWaveAssetPtr=&/door_residential_close_E3}}
]
properties {
 attenuationSettings=/Game/Sounds/Attenuation/Fort_Doors_Att
 bOverrideConcurrency=true
 maxDistance=2400
 soundClassObject=&Fort_Doors
 volumeMultiplier=0.6
 concurrencyOverrides {resolutionRule=PreventNew}
}
