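# Sound cue for a residential wood door closing (H variant), in the Fort_Doors sound class with door attenuation that has no occlusion.
# A quality-level root sends the first two levels through a modulator to an equally weighted random pick of three close recordings.
# The last two levels share a lighter modulated branch whose random node holds one recording and an empty slot, excluded from branch culling.
# A concurrency override prevents new instances once the limit is reached.
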SoundCue /Game/Sounds/Fort_Doors/World/door_residential_close_h_Cue
root=QualityLevel_0
nodes [
 {id=Modulator_0, children [Random_1]}
 {id=Modulator_2, children [Random_5]}
 {id=QualityLevel_0, children [Modulator_0,Modulator_0,Modulator_2,Modulator_2]}
 {id=Random_1, children [WavePlayer_7,WavePlayer_6,WavePlayer_8], properties {weights [1,1,1]}}
 {id=Random_5, children [WavePlayer_9,null], properties {bShouldExcludeFromBranchCulling=true, weights [1,1]}}
 {id=WavePlayer_6, properties {soundWaveAssetPtr=&/Residential_Wood_Door_H_Close_03}}
 {id=WavePlayer_7, properties {soundWaveAssetPtr=&/Residential_Wood_Door_H_Close_02}}
 {id=WavePlayer_8, properties {soundWaveAssetPtr=&/Residential_Wood_Door_H_Close_01}}
 {id=WavePlayer_9, properties {soundWaveAssetPtr=&Residential_Wood_Door_H_Close_01}}
]
properties {
 attenuationSettings=/Game/Sounds/Attenuation/Fort_Doors_NoOcclusion_Att
 bOverrideConcurrency=true
 maxDistance=2400
 soundClassObject=&Fort_Doors
 volumeMultiplier=0.6
 concurrencyOverrides {resolutionRule=PreventNew}
}
