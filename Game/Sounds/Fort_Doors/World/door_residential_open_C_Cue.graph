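# Sound cue for the variant C opening sound of a residential door, using the Fort_Doors sound class and door attenuation.
# A quality-level root sends its first two slots to a random pick of four waves and its last two to a pick of only C3 and C4, each through a Modulator.
# The two-wave random node is excluded from branch culling.
# Overridden concurrency blocks new instances at the limit.
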SoundCue /Game/Sounds/Fort_Doors/World/door_residential_open_C_Cue
root=QualityLevel_0
nodes [
 {id=Modulator_0, children [Random_1]}
 {id=Modulator_4, children [Random_2]}
 {id=QualityLevel_0, children [Modulator_0,Modulator_0,Modulator_4,Modulator_4]}
 {id=Random_1, children [WavePlayer_4,WavePlayer_5,WavePlayer_6,WavePlayer_7], properties {weights [1,1,1,1]}}
 {id=Random_2, children [WavePlayer_12, WavePlayer_13], properties {bShouldExcludeFromBranchCulling=true, weights [1,1]}}
 {id=WavePlayer_12, properties {soundWaveAssetPtr=&/door_residental_open_C3_athena}}
 {id=WavePlayer_13, properties {soundWaveAssetPtr=&/door_residental_open_C4_athena}}
 {id=WavePlayer_4, properties {soundWaveAssetPtr=&/door_residental_open_C1_athena}}
 {id=WavePlayer_5, properties {soundWaveAssetPtr=&/door_residental_open_C2_athena}}
 {id=WavePlayer_6, properties {soundWaveAssetPtr=&door_residental_open_C3_athena}}
 {id=WavePlayer_7, properties {soundWaveAssetPtr=&door_residental_open_C4_athena}}
]
properties {
 attenuationSettings=/Game/Sounds/Attenuation/Fort_Doors_Att
 bOverrideConcurrency=true
 maxDistance=2400
 soundClassObject=&Fort_Doors
 volumeMultiplier=0.4
 concurrencyOverrides {resolutionRule=PreventNew}
}
