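# Sound cue for a residential door opening (variant E) in the world, using the Fort_Doors sound class and attenuation.
# A quality-level node sends its first two slots to a modulated random pick of four waves and its last two to a modulated pick of only two.
# The two-wave random node is excluded from branch culling.
# The cue plays at reduced volume and overrides concurrency to prevent new instances at the limit.
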SoundCue /Game/Sounds/Fort_Doors/World/door_residential_open_E_Cue
root=QualityLevel_0
nodes [
 {id=Modulator_0, children [Random_6]}
 {id=Modulator_1, children [Random_1]}
 {id=QualityLevel_0, children [Modulator_0,Modulator_0,Modulator_1,Modulator_1]}
 {id=Random_1, children [WavePlayer_23, WavePlayer_24], properties {bShouldExcludeFromBranchCulling=true, weights [1,1]}}
 {id=Random_6, children [WavePlayer_19, WavePlayer_20, WavePlayer_21, WavePlayer_22], properties {weights [1,1,1,1]}}
 {id=WavePlayer_19, properties {soundWaveAssetPtr=&/door_residental_open_E1_athena}}
 {id=WavePlayer_20, properties {soundWaveAssetPtr=&/door_residental_open_E2_athena}}
 {id=WavePlayer_21, properties {soundWaveAssetPtr=&/door_residental_open_E3_athena}}
 {id=WavePlayer_22, properties {soundWaveAssetPtr=&/door_residental_open_E4_athena}}
 {id=WavePlayer_23, properties {soundWaveAssetPtr=&door_residental_open_E1_athena}}
 {id=WavePlayer_24, properties {soundWaveAssetPtr=&door_residental_open_E2_athena}}
]
properties {
 attenuationSettings=/Game/Sounds/Attenuation/Fort_Doors_Att
 bOverrideConcurrency=true
 maxDistance=2400
 soundClassObject=&Fort_Doors
 volumeMultiplier=0.4
 concurrencyOverrides {resolutionRule=PreventNew}
}
