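# Sound cue that plays a residential wood door opening sound in the world, under the Fort_Doors sound class.
# A quality level node sends its first two levels to a modulated, equally weighted random pick of three wave variations, and the last two to a single modulated wave.
# It uses a door attenuation asset named for no occlusion, reduced volume, and a concurrency override that prevents new instances once the limit is reached.
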SoundCue /Game/Sounds/Fort_Doors/World/door_residential_open_h_Cue
root=QualityLevel_0
nodes [
 {id=Modulator_0, children [Random_0]}
 {id=Modulator_1, children [WavePlayer_9]}
 {id=QualityLevel_0, children [Modulator_0,Modulator_0,Modulator_1,Modulator_1]}
 {id=Random_0, children [WavePlayer_6,WavePlayer_7,WavePlayer_8], properties {weights [1,1,1]}}
 {id=WavePlayer_6, properties {soundWaveAssetPtr=&/Residential_Wood_Door_H_Open_01}}
 {id=WavePlayer_7, properties {soundWaveAssetPtr=&/Residential_Wood_Door_H_Open_02}}
 {id=WavePlayer_8, properties {soundWaveAssetPtr=&/Residential_Wood_Door_H_Open_03}}
 {id=WavePlayer_9, properties {soundWaveAssetPtr=&Residential_Wood_Door_H_Open_01}}
]
properties {
 attenuationSettings=/Game/Sounds/Attenuation/Fort_Doors_NoOcclusion_Att
 bOverrideConcurrency=true
 maxDistance=2400
 soundClassObject=&Fort_Doors
 volumeMultiplier=0.4
 concurrencyOverrides {resolutionRule=PreventNew}
}
