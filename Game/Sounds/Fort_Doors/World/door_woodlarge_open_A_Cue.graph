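# Sound cue for the large wooden door opening sound, variant A, among the Fort door world sounds.
# A single wave player plays the door_woodlarge_open_A wave with no randomization or modulation, at reduced volume.
# It uses the shared Fort_Doors attenuation and sound class, and overrides concurrency so new instances are blocked rather than replacing playing ones once the limit is reached.
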SoundCue /Game/Sounds/Fort_Doors/World/door_woodlarge_open_A_Cue
root=WavePlayer_17
nodes [{id=WavePlayer_17, properties {soundWaveAssetPtr=&/door_woodlarge_open_A}}]
properties {
 attenuationSettings=/Game/Sounds/Attenuation/Fort_Doors_Att
 bOverrideConcurrency=true
 maxDistance=2400
 soundClassObject=&Fort_Doors
 volumeMultiplier=0.6
 concurrencyOverrides {resolutionRule=PreventNew}
}
