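# Sound cue for a medium wooden door closing, part of the Fort_Doors world sound set.
# A single wave player feeds one close sample straight to the output, with no randomization or modulation, at a reduced volume.
# It uses the shared Fort_Doors attenuation and sound class, and overrides concurrency to prevent new instances from starting once the limit is reached.
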SoundCue /Game/Sounds/Fort_Doors/World/door_woodmedium_close_A_Cue
root=WavePlayer_18
nodes [{id=WavePlayer_18, properties {soundWaveAssetPtr=&/door_woodmedium_close_A}}]
properties {
 attenuationSettings=/Game/Sounds/Attenuation/Fort_Doors_Att
 bOverrideConcurrency=true
 maxDistance=2400
 soundClassObject=&Fort_Doors
 volumeMultiplier=0.6
 concurrencyOverrides {resolutionRule=PreventNew}
}
