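# Sound cue for a lab door closing, in the Fort_Doors sound class with shared door attenuation.
# A quality-level root sends the first two levels to an equal-weight random pick between two close waves, and the last two to a single "opt" variant.
# An overridden concurrency rule blocks new instances once the limit is reached.
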SoundCue /Game/Sounds/Fort_Doors/World/lab_door_close_A_Cue
root=QualityLevel_0
nodes [
 {id=QualityLevel_0, children [Random_0,Random_0,WavePlayer_0,WavePlayer_0]}
 {id=Random_0, children [WavePlayer_1,WavePlayer_2], properties {weights [1,1]}}
 {id=WavePlayer_0, properties {soundWaveAssetPtr=&/Lab_Door_Close_02_opt}}
 {id=WavePlayer_1, properties {soundWaveAssetPtr=&/Lab_Door_Close_03}}
 {id=WavePlayer_2, properties {soundWaveAssetPtr=&/Lab_Door_Close_02}}
]
properties {
 attenuationSettings=/Game/Sounds/Attenuation/Fort_Doors_Att
 bOverrideConcurrency=true
 maxDistance=2400
 soundClassObject=&Fort_Doors
 volumeMultiplier=1
 concurrencyOverrides {resolutionRule=PreventNew}
}
