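# Sound cue for a lab door opening, part of the world door sounds.
# A Random node picks with equal weight between two recorded variations, Lab_Door_Open_01 and Lab_Door_Open_04.
# It plays through the Fort_Doors sound class and door attenuation, with its own concurrency override that prevents new instances once the limit is reached.
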SoundCue /Game/Sounds/Fort_Doors/World/lab_door_open_A_Cue
root=Random_0
nodes [
 {id=Random_0, children [WavePlayer_0,WavePlayer_1], properties {weights [1,1]}}
 {id=WavePlayer_0, properties {soundWaveAssetPtr=&/Lab_Door_Open_01}}
 {id=WavePlayer_1, properties {soundWaveAssetPtr=&/Lab_Door_Open_04}}
]
properties {
 attenuationSettings=/Game/Sounds/Attenuation/Fort_Doors_Att
 bOverrideConcurrency=true
 maxDistance=2400
 soundClassObject=&Fort_Doors
 volumeMultiplier=1
 concurrencyOverrides {resolutionRule=PreventNew}
}
