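# Sound cue for a world metal door closing, routed through the Fort_Doors sound class with no-occlusion door attenuation.
# A mixer layers an enveloped metal_door_open_C wave, shaped by a volume swell and fade and a rising pitch curve, with a delayed pair of close waves (A and C) mixed together.
# The half-second delay lands the close as the open layer fades.
# A concurrency override prevents new instances once the limit is reached.
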SoundCue /Game/Sounds/Fort_Doors/World/metal_door_close_A_Cue
root=Mixer_7
nodes [
 {id=Delay_9, children [Mixer_8], properties {delayMax=0.5, delayMin=0.5}}
 {
  id=Enveloper_13
  children [WavePlayer_22]
  properties {
   pitchCurve {editorCurveData {keys [{time=0}, <at 1.0802 1.1692>]}}
   volumeCurve {editorCurveData {keys [<at 0 0.1>,<at 0.2733 1.4005>,<at 0.5 1.008>,<at 0.7488 0>]}}
  }
 }
 {id=Mixer_7, children [Enveloper_13,Delay_9], properties {inputVolume [1,1]}}
 {id=Mixer_8, children [WavePlayer_0, WavePlayer_23], properties {inputVolume [1,1]}}
 {id=WavePlayer_0, properties {soundWaveAssetPtr=&/metal_door_close_A}}
 {id=WavePlayer_22, properties {soundWaveAssetPtr=&/metal_door_open_C}}
 {id=WavePlayer_23, properties {soundWaveAssetPtr=&/metal_door_close_C}}
]
properties {
 attenuationSettings=/Game/Sounds/Attenuation/Fort_Doors_NoOcclusion_Att
 bOverrideConcurrency=true
 maxDistance=2400
 soundClassObject=&Fort_Doors
 volumeMultiplier=0.6
 concurrencyOverrides {resolutionRule=PreventNew}
}
