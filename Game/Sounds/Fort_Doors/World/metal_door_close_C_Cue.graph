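# Sound cue for a metal door closing in the world, built from two layered waves.
# A mixer plays the door-open wave through an envelope that swells and fades volume while raising pitch, with the close wave delayed half a second to overlap its tail.
# It uses the Fort_Doors sound class and attenuation, with a concurrency override that prevents new instances at the limit.
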SoundCue /Game/Sounds/Fort_Doors/World/metal_door_close_C_Cue
root=Mixer_6
nodes [
 {id=Delay_8, children [WavePlayer_21], properties {delayMax=0.5, delayMin=0.5}}
 {
  id=Enveloper_12
  children [WavePlayer_20]
  properties {
   pitchCurve {editorCurveData {keys [{time=0}, <at 1.0802 1.1692>]}}
   volumeCurve {editorCurveData {keys [<at 0 0>,<at 0.2733 1.4005>,<at 0.5 1.008>,<at 0.7488 0>]}}
  }
 }
 {id=Mixer_6, children [Enveloper_12,Delay_8], properties {inputVolume [1,1]}}
 {id=WavePlayer_20, properties {soundWaveAssetPtr=&/metal_door_open_C}}
 {id=WavePlayer_21, properties {soundWaveAssetPtr=&/metal_door_close_C}}
]
properties {
 attenuationSettings=/Game/Sounds/Attenuation/Fort_Doors_Att
 bOverrideConcurrency=true
 maxDistance=2400
 soundClassObject=&Fort_Doors
 volumeMultiplier=1
 concurrencyOverrides {resolutionRule=PreventNew}
}
