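# Sound cue for a metal door opening, variant B, in the Fort_Doors world sound set.
# A single wave player feeds metal_door_open_B straight to the output, with no randomization or modulation.
# It uses the Fort_Doors sound class and shared door attenuation at reduced volume, and a concurrency override that blocks new instances once the limit is reached.
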SoundCue /Game/Sounds/Fort_Doors/World/metal_door_open_B_Cue
root=WavePlayer_4
nodes [{id=WavePlayer_4, properties {soundWaveAssetPtr=&/metal_door_open_B}}]
properties {
 attenuationSettings=/Game/Sounds/Attenuation/Fort_Doors_Att
 bOverrideConcurrency=true
 maxDistance=2400
 soundClassObject=&Fort_Doors
 volumeMultiplier=0.6
 concurrencyOverrides {resolutionRule=PreventNew}
}
