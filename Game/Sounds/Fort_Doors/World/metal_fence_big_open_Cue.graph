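# Sound cue for the opening of a big metal fence door in the world, in the Fort_Doors sound class.
# A random node picks one of two equally weighted wave recordings, then feeds it through a modulator at the root for pitch and volume variation.
# Playback uses the shared door attenuation asset at reduced volume, with a concurrency override that prevents new instances when the limit is reached.
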SoundCue /Game/Sounds/Fort_Doors/World/metal_fence_big_open_Cue
root=Modulator_3
nodes [
 {id=Modulator_3, children [Random_2]}
 {id=Random_2, children [WavePlayer_8,WavePlayer_9], properties {weights [1,1]}}
 {id=WavePlayer_8, properties {soundWaveAssetPtr=&/metal_fence_big_open_01}}
 {id=WavePlayer_9, properties {soundWaveAssetPtr=&/metal_fence_big_open_02}}
]
properties {
 attenuationSettings=/Game/Sounds/Attenuation/Fort_Doors_Att
 bOverrideConcurrency=true
 maxDistance=2400
 soundClassObject=&Fort_Doors
 volumeMultiplier=0.6
 concurrencyOverrides {resolutionRule=PreventNew}
}
