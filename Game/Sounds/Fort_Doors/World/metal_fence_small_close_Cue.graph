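# Sound cue for a small metal fence door closing in the world, assigned to the Fort_Doors sound class and attenuation.
# A single wave player plays the big metal fence close recording through a modulator that pins pitch at 1.4, reusing that sound at higher pitch for the smaller fence.
# Volume is reduced, and a concurrency override prevents new instances from starting once the limit is reached.
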SoundCue /Game/Sounds/Fort_Doors/World/metal_fence_small_close_Cue
root=Modulator_4
nodes [
 {id=Modulator_4, children [WavePlayer_10], properties {pitchMax=1.4, pitchMin=1.4}}
 {id=WavePlayer_10, properties {soundWaveAssetPtr=&/metal_fence_big_close_01}}
]
properties {
 attenuationSettings=/Game/Sounds/Attenuation/Fort_Doors_Att
 bOverrideConcurrency=true
 maxDistance=2400
 soundClassObject=&Fort_Doors
 volumeMultiplier=0.6
 concurrencyOverrides {resolutionRule=PreventNew}
}
