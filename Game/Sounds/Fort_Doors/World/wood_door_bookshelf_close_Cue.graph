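# Sound cue for a wooden bookshelf door closing in the world, played through the Fort_Doors sound class.
# A single wave player plays one close sample with no randomization, at slightly reduced volume.
# It uses the shared Fort_Doors attenuation and overrides concurrency so new instances are blocked once the limit is reached.
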SoundCue /Game/Sounds/Fort_Doors/World/wood_door_bookshelf_close_Cue
root=WavePlayer_1
nodes [{id=WavePlayer_1, properties {soundWaveAssetPtr=&/wood_door_bookshelf_close_01}}]
properties {
 attenuationSettings=/Game/Sounds/Attenuation/Fort_Doors_Att
 bOverrideConcurrency=true
 maxDistance=2400
 soundClassObject=&Fort_Doors
 volumeMultiplier=0.9
 concurrencyOverrides {resolutionRule=PreventNew}
}
