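# Sound cue for a small wooden fence door closing in the world, part of the Fort_Doors sound set.
# A single wave player plays the wood gate close sample through a modulator that slightly randomizes pitch for variation.
# It uses the doors attenuation and sound class at reduced volume, with a concurrency override that blocks new instances once the limit is reached.
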SoundCue /Game/Sounds/Fort_Doors/World/wood_fence_small_close_Cue
root=Modulator_4
nodes [
 {id=Modulator_4, children [WavePlayer_12], properties {pitchMax=1.1, pitchMin=0.9}}
 {id=WavePlayer_12, properties {soundWaveAssetPtr=&/wood_gate_close_01}}
]
properties {
 attenuationSettings=/Game/Sounds/Attenuation/Fort_Doors_Att
 bOverrideConcurrency=true
 maxDistance=2400
 soundClassObject=&Fort_Doors
 volumeMultiplier=0.6
 concurrencyOverrides {resolutionRule=PreventNew}
}
