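# Sound cue for the opening sound of a small wooden fence door or gate in the world.
# A single wood gate open wave plays through a modulator that slightly randomizes pitch and volume on each play.
# It uses the Fort_Doors sound class and door attenuation, with a concurrency override that prevents new instances rather than stopping playing ones.
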SoundCue /Game/Sounds/Fort_Doors/World/wood_fence_small_open_Cue
root=Modulator_3
nodes [
 {id=Modulator_3, children [WavePlayer_13], properties {pitchMax=1.1, pitchMin=0.9, volumeMax=0.5, volumeMin=0.4}}
 {id=WavePlayer_13, properties {soundWaveAssetPtr=&/wood_gate_open_01}}
]
properties {
 attenuationSettings=/Game/Sounds/Attenuation/Fort_Doors_Att
 bOverrideConcurrency=true
 maxDistance=2400
 soundClassObject=&Fort_Doors
 volumeMultiplier=0.6
 concurrencyOverrides {resolutionRule=PreventNew}
}
